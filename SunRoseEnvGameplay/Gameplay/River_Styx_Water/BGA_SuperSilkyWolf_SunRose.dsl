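; SunRose child Blueprint of a Super Silky Wolf actor, in the River Styx water gameplay folder, that applies and removes a "Silky GA" on actors.
; It overrides ApplySilkyGA and RemoveSilkyGA to call the parent versions only with network authority, so that logic runs server-side.
; BeginPlay and the construction script only forward to the parent, while Tick and BeginOverlap are empty, leaving the actual effect to the parent class.
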(event EventBeginPlay
  (|CallParentFunction))

(event EventTick
  (params (DeltaSeconds :type Float%20%28single-precision%29)))

(event Collision|EventActorBeginOverlap
  (params (OtherActor :type Actor%20Object%20Reference)))

(fn ApplySilkyGA
  (params (InActor :type Actor%20Object%20Reference))
  (Networking|SwitchHasAuthority
    (:Authority
      (|CallParentFunction InActor))))

(fn RemoveSilkyGA
  (params (InActor :type Actor%20Object%20Reference))
  (Networking|SwitchHasAuthority
    (:Authority
      (|CallParentFunction InActor))))

(fn ConstructionScript
  (|CallParentFunction))
