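; Game state component that adds the SunRose water body component to water bodies and waterfalls in the match.
; When playlist data is ready, the authority, gated by a scalable-float toggle, scans configured soft classes for actors tagged "SunRoseWater" and adds the component where missing.
; The authority also handles dynamic POI injection events, using a retriggerable delay and streaming-completion wait before reapplying.
; A cheat event applies it to every actor of a cheat class, skipping the tag check.
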(event EventBeginPlay
  ;[0 0.1667 0.02545] BeginPlay: Bind to know when a player logs into the match.
  ; Add SunRose water body component to the water bodies in the world that need it.
  (|CallParentFunction)
  (Utilities|FlowControl|Sequence
    (:then_0
      (Class|FortAsyncActionCurrentPlaylistReady|PlaylistDataReadyAsync
        (:OnReady
          (CallFunction|ApplySunRoseWaterComponenttoAllWaterBodies))))
    (:then_1
      ;[0 0.1667 0.02545] BeginPlay: Bind to know when a level is injected and loaded
      (Networking|SwitchHasAuthority
        (:Authority
          (bind _returnvalue (GameplayEventRouter|GetEventRouter "/Script/FortniteGame.FortGameState" (Components|GetOwner)))
          (GameplayEventRouter|StartListeningToEvent _returnvalue "/Script/FortLevelInjectingRuntime.OnFortDynamicPOIInjected"
            (:OnEventReceived
              (if (not (Variables|Default|GetWaitingForStreamingCompleted))
                (Utilities|FlowControl|RetriggerableDelay 2.0) ; Waits a bit to give WorldPartition time to process the injected level
                (Class|AsyncActionWaitforStreamingCompletedEvent|WaitforStreamingCompletedEvent
                  (:then
                    (Variables|Default|SetWaitingForStreamingCompleted true))
                  (:Completed
                    _)
                  (:Failed
                    _))
                (Variables|Default|SetWaitingForStreamingCompleted false)
                (CallFunction|ApplySunRoseWaterComponenttoAllWaterBodies)))))))))

(event Custom|SunRoseAllWaterBodies
  ;[0.2481 1 0.4642] Cheat: ChromeAllWaterBodies: Adds the chrome component to all the water bodies in the world.
  (bind _outactors (Actor|GetAllActorsOfClass (Variables|Default|GetCheat%20Water%20Body%20Actor%20Class)))
  (for _array_element _outactors
    (Utilities|IsValid _array_element
      (:Is%20Valid
        (CallFunction|SunRoseWaterBody _array_element)))))

(fn ApplySunRoseWaterComponentToAllWaterBodies
  (Networking|SwitchHasAuthority
    (:Authority
      (if (ScalableFloat|GetValueasBool (Variables|Default|GetWaterBodyEnabled))
        (for _array_element (Variables|Default|GetWater%20Bodies%20and%20Waterfalls%20Classes)
          (if (Utilities|IsValidClass (Utilities|ResolveSoftReference _array_element))
            (CallFunction|ValidateandApplySunRoseWaterComponent (Utilities|ResolveSoftReference _array_element))))))))

(fn SunRoseWaterBody
  (params (WaterBody :type Actor%20Object%20Reference))
  (bind _sunrosewaterbodycomponentclass (Variables|Default|GetSunRoseWaterBodyComponentClass))
  (Utilities|IsValid WaterBody
    (:Is%20Valid
      (Utilities|IsValid (Actor|GetComponentbyClass WaterBody _sunrosewaterbodycomponentclass)
        (:Is%20Not%20Valid
          (Game|AddComponentbyClass WaterBody _sunrosewaterbodycomponentclass))))))

(fn ValidateAndApplySunRoseWaterComponent
  (params (ActorClass :type Actor%20Class%20Reference))
  (bind _outactors (Actor|GetAllActorsOfClass ActorClass))
  (for _array_element _outactors
    (Utilities|IsValid _array_element
      (:Is%20Valid
        (if (Utilities|Array|ContainsItem (Variables|/Script/Engine.Actor|GetTags _array_element) "SunRoseWater")
          (CallFunction|SunRoseWaterBody _array_element))))))
